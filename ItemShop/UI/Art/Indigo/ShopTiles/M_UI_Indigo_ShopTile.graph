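# UI-domain material for Indigo shop tiles in the item shop, with an IndigoTexture parameter defaulting to a sunset texture.
# A UV chain built on GetUserInterfaceUV scales and centers each axis, then lerps between the two results by a pixel-size comparison, resembling aspect-ratio fitting.
# As wired, each ratio and comparison uses one component against itself, so the math passes UVs through unchanged.
# The graph shows no links from that chain to the texture sample or to material outputs.
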Material /ItemShop/UI/Art/Indigo/ShopTiles/M_UI_Indigo_ShopTile
graph {
 properties {bUsedWithStaticMesh=false, materialDomain=MD_UI, pixelDepthOffsetMode=PDOM_Legacy, usedShadingModels=Default Lit}
 nodes [
  n1 = MaterialFunctionCall() {materialFunction=&GetUserInterfaceUV}
  n2 = MaterialFunctionCall(Float2=@n1.Pixel Size) {materialFunction=&BreakOutFloat2Components}
  n3 = Divide(A=@n2.G, B=@n2.G)
  n4 = AppendVector(A=Constant() {r=1}, B=@n3)
  n5 = Constant()
  n6 = Multiply(A=OneMinus(None=@n3)) {constB=0.5}
  n7 = MaterialFunctionCall(Float2=MaterialFunctionCall() {materialFunction=&GetUserInterfaceUV}.Pixel Size) {materialFunction=&BreakOutFloat2Components}
  n8 = Constant()
  n9 = MaterialFunctionCall() {materialFunction=&GetUserInterfaceUV}
  n10 = MaterialFunctionCall(Float2=@n9.Pixel Size) {materialFunction=&BreakOutFloat2Components}
  n11 = Divide(A=@n10.R, B=@n10.R)
  n12 = AppendVector(A=@n11, B=Constant() {r=1})
  n13 = Multiply(A=OneMinus(None=@n11)) {constB=0.5}
  n14 = Constant()
  LinearInterpolate(
   A=Add(A=Multiply(A=@n1.Normalized UV, B=@n4), B=AppendVector(A=@n5, B=@n6))
   B=Add(A=Multiply(A=@n9.Normalized UV, B=@n12), B=AppendVector(A=@n13, B=@n14))
   Alpha=If(A=@n7.R, B=@n7.R, "A > B"=@n8, "A == B"=@n8, "A < B"=Constant() {r=1})
  )
  TextureSampleParameter2D() {parameterName=IndigoTexture, texture=/OfferCatalog/Art/Indigo/ShopTiles/Textures/T_UI_Sunset}
 ]
}
